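# Physics asset for the collar mesh of the medium male Mummy player body.
# It defines no constraints, so it sets up no jointed ragdoll or secondary motion between bodies.
# No physics bodies or collision shapes are listed, so it appears to be an empty or placeholder asset.
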PhysicsAsset /Game/Characters/Player/Male/Medium/Bodies/M_MED_Mummy/Meshes/M_MED_Mummy_Collar_Physics
constraints []
