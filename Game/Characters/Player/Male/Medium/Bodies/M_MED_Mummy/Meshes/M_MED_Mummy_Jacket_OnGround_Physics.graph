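# Physics asset for the Mummy male player body's OnGround jacket mesh, providing bodies and joints for dynamic jacket flaps and straps.
# Spheres and capsules cover the pelvis, spine, thighs and left arm, plus dyn_ bones for jacket panels, front strap and left-hand strap.
# Angle-limited chains hang front panels from spine_04, the front strap from spine_05, and the hand strap from hand_l.
# The left side panel and left-arm joints have free swing and twist.
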PhysicsAsset /Game/Characters/Player/Male/Medium/Bodies/M_MED_Mummy/Meshes/M_MED_Mummy_Jacket_OnGround_Physics
bodies {
 dyn_front_strap_01 [{shapeType=Sphere, radius=2, shapeName=dyn_front_strap_01_sphere, center=<1.7657 -1.9398 0.377>}]
 dyn_front_strap_02 [{
   shapeType=Capsule
   length=4.6734
   radius=1.9665
   shapeName=dyn_front_strap_02_capsule
   center=<1.5673 -1.3339 -0.158>
   rotation=<rot 0 -92.1724 -90.0001>
  }]
 dyn_front_strap_03 [{
   shapeType=Capsule
   length=5.005
   radius=2.005
   shapeName=dyn_front_strap_03_capsule
   center=<2.6106 -0.8044 0>
   rotation=<rot -80 179.9995 179.9998>
  }]
 dyn_hand_l_strap_01 [{
   shapeType=Capsule
   length=1.2263
   radius=1.8451
   shapeName=dyn_hand_l_strap_01_capsule
   center=<0.8894 -0.0881 0.6739>
   rotation=<rot 80 -179.9993 -179.9994>
  }]
 dyn_hand_l_strap_02 [{shapeType=Sphere, radius=1.8854, shapeName=dyn_hand_l_strap_02_sphere, center=<1.3316 0.0799 0.6524>}]
 dyn_jacket_forward_bottom_l_01 [{
   shapeType=Capsule
   length=8.5
   radius=3
   shapeName=dyn_jacket_forward_bottom_l_01_capsule
   center=<-0.225 -1.4001 0>
   rotation=<rot -90 179.9998 -180>
  }]
 dyn_jacket_forward_bottom_r_01 [{
   shapeType=Capsule
   length=5
   radius=3
   shapeName=dyn_jacket_forward_bottom_r_01_capsule
   center=<2.6191 -1.1573 0>
   rotation=<rot -90 0 -0.0003>
  }]
 dyn_jacket_front_l_01 [{
   shapeType=Capsule
   length=11.5
   radius=3
   shapeName=dyn_jacket_front_l_01_capsule
   center=<6.3435 -0.6069 -3.5998>
   rotation=<rot -9.8461 -1.7539 4.7825>
  }]
 dyn_jacket_front_r_01 [
  {shapeType=Sphere, radius=3.5, shapeName=dyn_jacket_front_r_01_sphere, center=<12.9657 -3.4711 2.0082>}
  {
   shapeType=Capsule
   length=11
   radius=2.5
   shapeName=dyn_jacket_front_r_01_capsule
   center=<5.7298 0.4218 2.096>
   rotation=<rot -0.0001 -0.0001 5>
  }
 ]
 dyn_jacket_side_l_01 [{shapeType=Sphere, radius=8.3773, shapeName=dyn_jacket_side_l_01_sphere, center=<-6.8846 0.1142 7.2231>}]
 hand_l [{
   shapeType=Capsule
   length=8.472
   radius=5.661
   shapeName=hand_l_capsule
   center=<4.0814 -2.7647 -0.2231>
   rotation=<rot 90 179.4936 179.4934>
  }]
 lowerarm_l [{shapeType=Capsule, length=7.5843, radius=5.1826, shapeName=lowerarm_l_capsule, center=<13.9254 0.1833 -0.7117>, rotation=<rot 90 0 0>}]
 pelvis [{shapeType=Capsule, length=9, radius=11.2745, shapeName=pelvis_capsule, center=<0.9381 -0.91 0.2335>}]
 spine_01 [{shapeType=Capsule, length=7.595, radius=8.3736, shapeName=spine_01_capsule, center=<5.8733 1.6601 -0.2913>}]
 spine_02 [{shapeType=Capsule, length=6.5074, radius=8.4227, shapeName=spine_02_capsule, center=<4.2613 3.0164 -0.0142>}]
 spine_03 [{shapeType=Capsule, length=6.9073, radius=10.9535, shapeName=spine_03_capsule, center=<5.4143 1.6519 0.2103>}]
 spine_04 [{shapeType=Sphere, radius=11.7832, center=<2.9028 1.1442 -2.3175>}]
 spine_05 [{shapeType=Sphere, radius=12.8649, center=<-1.3236 0.4303 0.0934>}]
 thigh_l [{
   shapeType=Capsule
   length=26.1654
   radius=7.7127
   shapeName=thigh_l_capsule
   center=<-10.9728 0.0713 0.4249>
   rotation=<rot 90 89.6476 89.6481>
  }]
 thigh_r [{
   shapeType=Capsule
   length=24.5242
   radius=8.8644
   shapeName=thigh_r_capsule
   center=<10.541 1.7711 -0.7875>
   rotation=<rot 90 89.6476 88.0006>
  }]
}
constraints [
 {bone1Name=spine_01, bone2Name=pelvis, swing1LimitDegrees=45, swing2LimitDegrees=45, twistLimitDegrees=45}
 {bone1Name=spine_02, bone2Name=spine_01, swing1LimitDegrees=45, swing2LimitDegrees=45, twistLimitDegrees=45}
 {bone1Name=spine_03, bone2Name=spine_02, swing1LimitDegrees=45, swing2LimitDegrees=45, twistLimitDegrees=45}
 {bone1Name=spine_04, bone2Name=spine_03, swing1LimitDegrees=45, swing2LimitDegrees=45, twistLimitDegrees=45}
 {bone1Name=spine_05, bone2Name=spine_04, swing1LimitDegrees=45, swing2LimitDegrees=45, twistLimitDegrees=45}
 {bone1Name=dyn_jacket_front_r_01, bone2Name=spine_04, swing1LimitDegrees=30, swing2LimitDegrees=10, twistLimitDegrees=2}
 {bone1Name=thigh_l, bone2Name=pelvis, swing1LimitDegrees=45, swing2LimitDegrees=45, twistLimitDegrees=45}
 {bone1Name=thigh_r, bone2Name=pelvis, swing1LimitDegrees=45, swing2LimitDegrees=45, twistLimitDegrees=45}
 {bone1Name=dyn_jacket_front_l_01, bone2Name=spine_04, swing1LimitDegrees=40, swing2LimitDegrees=30, twistLimitDegrees=15}
 {
  bone1Name=dyn_jacket_forward_bottom_l_01
  bone2Name=dyn_jacket_front_l_01
  swing1LimitDegrees=25
  swing2LimitDegrees=10
  twistLimitDegrees=15
 }
 {bone1Name=dyn_jacket_forward_bottom_r_01, bone2Name=dyn_jacket_front_r_01, swing1LimitDegrees=15, swing2LimitDegrees=5, twistLimitDegrees=5}
 {bone1Name=dyn_front_strap_01, bone2Name=spine_05, swing1LimitDegrees=50, swing2LimitDegrees=50, twistLimitDegrees=2}
 {bone1Name=dyn_front_strap_02, bone2Name=dyn_front_strap_01, swing1LimitDegrees=50, swing2LimitDegrees=50, twistLimitDegrees=2}
 {bone1Name=dyn_front_strap_03, bone2Name=dyn_front_strap_02, swing1LimitDegrees=50, swing2LimitDegrees=50, twistLimitDegrees=2}
 {bone1Name=dyn_hand_l_strap_01, bone2Name=hand_l, swing1LimitDegrees=40, swing2LimitDegrees=40, twistLimitDegrees=40}
 {bone1Name=dyn_hand_l_strap_02, bone2Name=dyn_hand_l_strap_01, swing1LimitDegrees=60, swing2LimitDegrees=60, twistLimitDegrees=40}
 {
  bone1Name=hand_l
  bone2Name=lowerarm_l
  swing1LimitDegrees=45
  swing1Motion=Free
  swing2LimitDegrees=45
  swing2Motion=Free
  twistLimitDegrees=45
  twistMotion=Free
 }
 {
  bone1Name=lowerarm_l
  bone2Name=spine_05
  swing1LimitDegrees=45
  swing1Motion=Free
  swing2LimitDegrees=45
  swing2Motion=Free
  twistLimitDegrees=45
  twistMotion=Free
 }
 {
  bone1Name=dyn_jacket_side_l_01
  bone2Name=spine_04
  swing1LimitDegrees=45
  swing1Motion=Free
  swing2LimitDegrees=45
  swing2Motion=Free
  twistLimitDegrees=45
  twistMotion=Free
 }
]
