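# Physics asset for the male medium Mummy skydiving jacket, providing secondary-motion bodies for its panels and front strap.
# Short capsule chains for the back, side, front and pec panels hang from spine_04 with tight angular limits, while a four-link strap chain hangs from spine_05.
# Extra free-rotation constraints link neighboring upper panels from front-left around the back to front-right.
# Larger pelvis, spine and thigh shapes approximate the body.
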PhysicsAsset /Game/Characters/Player/Male/Medium/Bodies/M_MED_Mummy/Meshes/M_MED_Mummy_Jacket_Skydiving_Physics
bodies {
 dyn_front_strap_01 [{shapeType=Sphere, radius=2, shapeName=dyn_front_strap_01_sphere, center=<1.7657 -3.3388 0>}]
 dyn_front_strap_02 [{shapeType=Capsule, length=4.1734, radius=2.4665, shapeName=dyn_front_strap_02_capsule, center=<1.566 -2.3287 0>, rotation=<rot 0 -90 -90>}]
 dyn_front_strap_03 [{shapeType=Sphere, radius=2.5, shapeName=dyn_front_strap_03_sphere, center=<0.4152 -2.3122 0>}]
 dyn_front_strap_04 [{shapeType=Sphere, radius=3, shapeName=dyn_front_strap_04_sphere, center=<0.7022 -1.8337 0>}]
 dyn_jacket_back_01 [{shapeType=Capsule, length=5, radius=3, shapeName=dyn_jacket_back_01_capsule, center=<2.6191 -1.9175 0>}]
 dyn_jacket_back_02 [{shapeType=Capsule, length=5, radius=3, shapeName=dyn_jacket_back_02_capsule, center=<2.6191 -0.8802 0>}]
 dyn_jacket_back_side_l_01 [{shapeType=Capsule, length=5, radius=3, shapeName=dyn_jacket_back_side_l_01_capsule, center=<2.6191 0 0>}]
 dyn_jacket_back_side_l_02 [{shapeType=Capsule, length=5, radius=3, shapeName=dyn_jacket_back_side_l_02_capsule, center=<2.6191 0 0>}]
 dyn_jacket_back_side_r_01 [{shapeType=Capsule, length=5, radius=3, shapeName=dyn_jacket_back_side_r_01_capsule, center=<2.6191 0 0>}]
 dyn_jacket_back_side_r_02 [{shapeType=Capsule, length=5, radius=3, shapeName=dyn_jacket_back_side_r_02_capsule, center=<2.6191 0 0>}]
 dyn_jacket_forward_bottom_l_01 [{
   shapeType=Capsule
   length=5
   radius=3
   shapeName=dyn_jacket_forward_bottom_l_01_capsule
   center=<2.6191 0 0>
   rotation=<rot -90 179.9999 179.9999>
  }]
 dyn_jacket_forward_bottom_r_01 [{
   shapeType=Capsule
   length=5
   radius=3
   shapeName=dyn_jacket_forward_bottom_r_01_capsule
   center=<2.6191 0 0>
   rotation=<rot -90 180 179.9997>
  }]
 dyn_jacket_forward_pecs_l_01 [{
   shapeType=Capsule
   length=5
   radius=3
   shapeName=dyn_jacket_forward_pecs_l_01_capsule
   center=<1.467 -0.1074 0>
   rotation=<rot -90 180 179.9998>
  }]
 dyn_jacket_forward_pecs_l_02 [{
   shapeType=Capsule
   length=5
   radius=3
   shapeName=dyn_jacket_forward_pecs_l_02_capsule
   center=<1.467 -0.1074 0>
   rotation=<rot -90 180 179.9998>
  }]
 dyn_jacket_forward_pecs_r_01 [{
   shapeType=Capsule
   length=5
   radius=3
   shapeName=dyn_jacket_forward_pecs_r_01_capsule
   center=<1.467 -0.1074 0>
   rotation=<rot -90 0 -0.0002>
  }]
 dyn_jacket_forward_pecs_r_02 [{
   shapeType=Capsule
   length=5
   radius=3
   shapeName=dyn_jacket_forward_pecs_r_02_capsule
   center=<1.467 -0.1074 0>
   rotation=<rot -90 0 -0.0002>
  }]
 dyn_jacket_front_l_01 [{shapeType=Capsule, length=5, radius=3, shapeName=dyn_jacket_front_l_01_capsule, center=<2.6191 -0.3627 0>, rotation=<rot 0.0002 -0.0001 10>}]
 dyn_jacket_front_l_02 [{shapeType=Capsule, length=5, radius=3, shapeName=dyn_jacket_front_l_02_capsule, center=<2.6191 0 0>}]
 dyn_jacket_front_r_01 [{shapeType=Capsule, length=5, radius=3, shapeName=dyn_jacket_front_r_01_capsule, center=<2.6191 -0.3627 0>}]
 dyn_jacket_front_r_02 [{shapeType=Capsule, length=5, radius=3, shapeName=dyn_jacket_front_r_02_capsule, center=<2.6191 0 0>}]
 dyn_jacket_side_l_01 [{shapeType=Capsule, length=5, radius=3, shapeName=dyn_jacket_side_l_01_capsule, center=<2.6191 -0.1792 0>}]
 dyn_jacket_side_l_02 [{shapeType=Capsule, length=5, radius=3, shapeName=dyn_jacket_side_l_02_capsule, center=<2.6191 0 0>}]
 dyn_jacket_side_r_01 [{shapeType=Capsule, length=5, radius=3, shapeName=dyn_jacket_side_r_01_capsule, center=<2.6191 -0.1792 0>}]
 dyn_jacket_side_r_02 [{shapeType=Capsule, length=5, radius=3, shapeName=dyn_jacket_side_r_02_capsule, center=<2.6191 0 0>}]
 pelvis [{shapeType=Capsule, length=7.5, radius=11.2745, shapeName=pelvis_capsule, center=<0.9381 -0.91 0.2335>}]
 spine_01 [{shapeType=Sphere, radius=11.095, center=<4.21 1.6601 -0.2913>}]
 spine_02 [{shapeType=Sphere, radius=11.0074, center=<4.1201 1.8717 -0.0142>}]
 spine_03 [{shapeType=Sphere, radius=11.2522, center=<6.6689 1.4216 0.6448>}]
 spine_04 [{shapeType=Sphere, radius=11.7832, center=<2.9028 1.7944 0.4777>}]
 spine_05 [{shapeType=Sphere, radius=12.8649, center=<-1.3236 0.4303 0.0934>}]
 thigh_l [{shapeType=Capsule, length=25.1654, radius=8.7127, shapeName=thigh_l_capsule, center=<-10.9728 0.0713 0.4249>, rotation=<rot 90 180 180>}]
 thigh_r [{shapeType=Capsule, length=24.5242, radius=8.8644, shapeName=thigh_r_capsule, center=<10.541 0.3254 -0.7875>, rotation=<rot 90 180 180>}]
}
constraints [
 {bone1Name=spine_01, bone2Name=pelvis, swing1LimitDegrees=45, swing2LimitDegrees=45, twistLimitDegrees=45}
 {bone1Name=spine_02, bone2Name=spine_01, swing1LimitDegrees=45, swing2LimitDegrees=45, twistLimitDegrees=45}
 {bone1Name=spine_03, bone2Name=spine_02, swing1LimitDegrees=45, swing2LimitDegrees=45, twistLimitDegrees=45}
 {bone1Name=spine_04, bone2Name=spine_03, swing1LimitDegrees=45, swing2LimitDegrees=45, twistLimitDegrees=45}
 {bone1Name=spine_05, bone2Name=spine_04, swing1LimitDegrees=45, swing2LimitDegrees=45, twistLimitDegrees=45}
 {bone1Name=dyn_jacket_front_r_01, bone2Name=spine_04, swing1LimitDegrees=25, swing2LimitDegrees=10, twistLimitDegrees=15}
 {bone1Name=dyn_jacket_front_r_02, bone2Name=dyn_jacket_front_r_01, swing1LimitDegrees=25, swing2LimitDegrees=10, twistLimitDegrees=15}
 {bone1Name=dyn_jacket_side_r_02, bone2Name=dyn_jacket_side_r_01, swing1LimitDegrees=25, swing2LimitDegrees=10, twistLimitDegrees=15}
 {bone1Name=dyn_jacket_back_side_r_02, bone2Name=dyn_jacket_back_side_r_01, swing1LimitDegrees=25, swing2LimitDegrees=10, twistLimitDegrees=15}
 {bone1Name=dyn_jacket_back_02, bone2Name=dyn_jacket_back_01, swing1LimitDegrees=25, swing2LimitDegrees=10, twistLimitDegrees=15}
 {bone1Name=dyn_jacket_back_side_l_01, bone2Name=spine_04, swing1LimitDegrees=25, swing2LimitDegrees=10, twistLimitDegrees=15}
 {bone1Name=dyn_jacket_side_l_01, bone2Name=spine_04, swing1LimitDegrees=25, swing2LimitDegrees=10, twistLimitDegrees=15}
 {bone1Name=dyn_jacket_side_l_02, bone2Name=dyn_jacket_side_l_01, swing1LimitDegrees=25, swing2LimitDegrees=10, twistLimitDegrees=15}
 {bone1Name=thigh_l, bone2Name=pelvis, swing1LimitDegrees=45, swing2LimitDegrees=45, twistLimitDegrees=45}
 {bone1Name=thigh_r, bone2Name=pelvis, swing1LimitDegrees=45, swing2LimitDegrees=45, twistLimitDegrees=45}
 {bone1Name=dyn_jacket_back_side_r_01, bone2Name=spine_04, swing1LimitDegrees=25, swing2LimitDegrees=10, twistLimitDegrees=15}
 {bone1Name=dyn_jacket_back_side_l_02, bone2Name=dyn_jacket_back_side_l_01, swing1LimitDegrees=25, swing2LimitDegrees=10, twistLimitDegrees=15}
 {bone1Name=dyn_jacket_back_01, bone2Name=spine_04, swing1LimitDegrees=25, swing2LimitDegrees=10, twistLimitDegrees=15}
 {bone1Name=dyn_jacket_forward_pecs_l_01, bone2Name=spine_04, swing1LimitDegrees=25, swing2LimitDegrees=10, twistLimitDegrees=15}
 {
  bone1Name=dyn_jacket_forward_pecs_l_02
  bone2Name=dyn_jacket_forward_pecs_l_01
  swing1LimitDegrees=25
  swing2LimitDegrees=10
  twistLimitDegrees=15
 }
 {bone1Name=dyn_jacket_forward_pecs_r_01, bone2Name=spine_04, swing1LimitDegrees=25, swing2LimitDegrees=10, twistLimitDegrees=15}
 {
  bone1Name=dyn_jacket_forward_pecs_r_02
  bone2Name=dyn_jacket_forward_pecs_r_01
  swing1LimitDegrees=25
  swing2LimitDegrees=10
  twistLimitDegrees=15
 }
 {bone1Name=dyn_jacket_front_l_01, bone2Name=spine_04, swing1LimitDegrees=25, swing2LimitDegrees=10, twistLimitDegrees=15}
 {
  bone1Name=dyn_jacket_forward_bottom_l_01
  bone2Name=dyn_jacket_front_l_01
  swing1LimitDegrees=25
  swing2LimitDegrees=10
  twistLimitDegrees=15
 }
 {bone1Name=dyn_jacket_front_l_02, bone2Name=dyn_jacket_front_l_01, swing1LimitDegrees=25, swing2LimitDegrees=10, twistLimitDegrees=15}
 {
  bone1Name=dyn_jacket_forward_bottom_r_01
  bone2Name=dyn_jacket_front_r_01
  swing1LimitDegrees=25
  swing2LimitDegrees=10
  twistLimitDegrees=15
 }
 {bone1Name=dyn_jacket_side_r_01, bone2Name=spine_04, swing1LimitDegrees=25, swing2LimitDegrees=10, twistLimitDegrees=15}
 {bone1Name=dyn_front_strap_01, bone2Name=spine_05, swing1LimitDegrees=30, swing2LimitDegrees=30, twistLimitDegrees=5}
 {bone1Name=dyn_front_strap_02, bone2Name=dyn_front_strap_01, swing1LimitDegrees=30, swing2LimitDegrees=30, twistLimitDegrees=5}
 {bone1Name=dyn_front_strap_03, bone2Name=dyn_front_strap_02, swing1LimitDegrees=30, swing2LimitDegrees=30, twistLimitDegrees=5}
 {bone1Name=dyn_front_strap_04, bone2Name=dyn_front_strap_03, swing1LimitDegrees=30, swing2LimitDegrees=30, twistLimitDegrees=5}
 {
  bone1Name=dyn_jacket_back_01
  bone2Name=dyn_jacket_back_side_l_01
  swing1LimitDegrees=45
  swing1Motion=Free
  swing2LimitDegrees=45
  swing2Motion=Free
  twistLimitDegrees=45
  twistMotion=Free
 }
 {
  bone1Name=dyn_jacket_back_01
  bone2Name=dyn_jacket_back_side_r_01
  swing1LimitDegrees=45
  swing1Motion=Free
  swing2LimitDegrees=45
  swing2Motion=Free
  twistLimitDegrees=45
  twistMotion=Free
 }
 {
  bone1Name=dyn_jacket_back_side_l_01
  bone2Name=dyn_jacket_side_l_01
  swing1LimitDegrees=45
  swing1Motion=Free
  swing2LimitDegrees=45
  swing2Motion=Free
  twistLimitDegrees=45
  twistMotion=Free
 }
 {
  bone1Name=dyn_jacket_side_l_01
  bone2Name=dyn_jacket_front_l_01
  swing1LimitDegrees=45
  swing1Motion=Free
  swing2LimitDegrees=45
  swing2Motion=Free
  twistLimitDegrees=45
  twistMotion=Free
 }
 {
  bone1Name=dyn_jacket_back_side_r_01
  bone2Name=dyn_jacket_side_r_01
  swing1LimitDegrees=45
  swing1Motion=Free
  swing2LimitDegrees=45
  swing2Motion=Free
  twistLimitDegrees=45
  twistMotion=Free
 }
 {
  bone1Name=dyn_jacket_side_r_01
  bone2Name=dyn_jacket_front_r_01
  swing1LimitDegrees=45
  swing1Motion=Free
  swing2LimitDegrees=45
  swing2Motion=Free
  twistLimitDegrees=45
  twistMotion=Free
 }
]
