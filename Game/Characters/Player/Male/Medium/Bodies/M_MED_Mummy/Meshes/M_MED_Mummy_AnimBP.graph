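# Post-process style AnimBP for the male medium Mummy player body, adding secondary motion to its jacket, collar, straps, pouch and zipper.
# It copies the parent mesh pose through a Control Rig into RigidBody jacket simulation.
# Cached-pose stages then layer collar and aim corrective bone offsets, driven by arm-raise, spine and targeting alphas.
# Chained AnimDynamics nodes add spring and gravity motion, before crouch-driven jacket offsets reach the output.
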AnimBlueprint /Game/Characters/Player/Male/Medium/Bodies/M_MED_Mummy/Meshes/M_MED_Mummy_AnimBP
target_skeleton=&/M_MED_Mummy_Skeleton
graphs [
 AnimGraph {nodes [
   n1 = Anim Dynamics - Bone: dyn_hand_l_strap_01(Alpha=0.7, ExternalForce="0,0,0") {
    angularBiasOverride=1
    angularSpringConstant=10000
    bAngularSpring=true
    bOverrideAngularBias=true
    bUseGravityOverride=true
    simulationSpace=RootRelative
    alphaBoolBlend {blendOption=Linear}
    boundBone {boneName=dyn_hand_l_strap_01}
    componentAppliedLinearAccClamp=<1000 1000 1000>
    componentLinearAccScale=<0.1 0.1 0.1>
    componentLinearVelScale=<0.1 0.1 0.1>
    gravityOverride {z=-200}
    physicsBodyDefinitions [{
      boundBone {boneName=dyn_hand_l_strap_01}
      boxExtents=<40 40 40>
      constraintSetup {angularLimitsMax=<30 50 30>, angularLimitsMin=<-30 -50 -30>, angularTarget=<0 0 0>}
      localJointOffset=<20 0 0>
     }]
    planarLimits [{drivingBone {boneName=hand_l}, planeTransform {location=<0 2.3 0>, rotation=<rot 0 10 90>}}]
   }
   Anim Dynamics - Bone: dyn_hand_l_strap_02(ComponentPose=@n1, Alpha=0.7, ExternalForce="0,0,0") {
    angularBiasOverride=1
    angularSpringConstant=10000
    bAngularSpring=true
    bOverrideAngularBias=true
    bUseGravityOverride=true
    simulationSpace=RootRelative
    alphaBoolBlend {blendOption=Linear}
    boundBone {boneName=dyn_hand_l_strap_02}
    componentAppliedLinearAccClamp=<1000 1000 1000>
    componentLinearAccScale=<0.01 0.01 0.01>
    componentLinearVelScale=<0.01 0.01 0.01>
    gravityOverride {z=-200}
    physicsBodyDefinitions [{
      boundBone {boneName=dyn_hand_l_strap_02}
      boxExtents=<40 40 40>
      constraintSetup {angularLimitsMax=<0 40 15>, angularLimitsMin=<0 -40 -15>, angularTarget=<0 0 0>}
      localJointOffset=<20 0 0>
     }]
    planarLimits [{drivingBone {boneName=hand_l}, planeTransform {location=<0 1.8 0>, rotation=<rot 0 10 90>}}]
   }
   n2 = RigidBody(ComponentPose=Local To Component(LocalPose=Use cached pose 'Pre-rigid body'())) {
    overridePhysicsAsset=&/M_MED_Mummy_Jacket_Skydiving_Physics
    simulationSpace=BaseBoneSpace
    alphaBoolBlend {blendOption=Linear}
    baseBoneRef {boneName=spine_01}
    componentAppliedLinearAccClamp=<2000 2000 2000>
    componentLinearAccScale=<0.3 0.3 0.3>
    componentLinearVelScale=<0.3 0.3 0.3>
   }
   Local To Component(LocalPose=Blend Poses by bool(BlendPose_0=Component To Local(ComponentPose=@n2), BlendPose_1=Component To Local()) {blendType=ExpInOut})
   "Transform (Modify) Bone - Bone: dyn_zipper_MSTR"!(ComponentPose, bAlphaBoolEnabled:bool=True, Translation="0,0,0", Rotation="0,-20,0", Scale="1,1,1") -> (Pose) {
    alphaInputType=Bool
    rotationMode=BMM_Additive
    rotationSpace=BCS_BoneSpace
    translationMode=BMM_Additive
    translationSpace=BCS_BoneSpace
    alphaBoolBlend {blendInTime=0.3, blendOption=ExpInOut, blendOutTime=0.3}
    alphaScaleBiasClamp {bInterpResult=true}
    boneToModify {boneName=dyn_zipper_MSTR}
    rotation {yaw=-20}
   }
   RigidBody() {
    overridePhysicsAsset=&/M_MED_Mummy_Collar_Physics
    simulationSpace=BaseBoneSpace
    alphaBoolBlend {blendOption=Linear}
    baseBoneRef {boneName=spine_01}
    componentAppliedLinearAccClamp=<2000 2000 2000>
    componentLinearAccScale=<0.3 0.3 0.3>
    componentLinearVelScale=<0.3 0.3 0.3>
   }
   n3 = "Transform (Modify) Bone - Bone: dyn_jacket_front_r_01"(
    ComponentPose=Local To Component(LocalPose=Use cached pose 'Post Anim Dynamics'())
    Alpha=Get Crouching to the left Alpha() -> (Crouching to the left Alpha:double)
    Translation="0,0,0"
    Rotation="0,-20,0"
    Scale="1,1,1"
   ) {
    rotationMode=BMM_Additive
    rotationSpace=BCS_BoneSpace
    alphaBoolBlend {blendOption=Linear}
    alphaScaleBiasClamp {bInterpResult=true}
    boneToModify {boneName=dyn_jacket_front_r_01}
    rotation {yaw=-20}
   }
   n4 = "Transform (Modify) Bone - Bone: dyn_jacket_front_r_01"(
    ComponentPose=@n3
    Alpha="Get Crouching forward of  backwards Alpha"() -> ("Crouching forward of  backwards Alpha":double)
    Translation="0,0,0"
    Rotation="0,-10,0"
    Scale="1,1,1"
   ) {
    rotationMode=BMM_Additive
    rotationSpace=BCS_BoneSpace
    alphaBoolBlend {blendOption=Linear}
    alphaScaleBiasClamp {bInterpResult=true}
    boneToModify {boneName=dyn_jacket_front_r_01}
    rotation {yaw=-10}
   }
   n5 = "Transform (Modify) Bone - Bone: dyn_jacket_front_l_01"!(ComponentPose=@n4, bAlphaBoolEnabled:bool=True, Translation="0,-5,0", Rotation="0,10,0", Scale="1,1,1") -> (Pose) {
    alphaInputType=Bool
    rotationMode=BMM_Additive
    rotationSpace=BCS_BoneSpace
    translationMode=BMM_Additive
    translationSpace=BCS_BoneSpace
    alphaBoolBlend {blendInTime=0.3, blendOption=ExpInOut, blendOutTime=0.3}
    alphaScaleBiasClamp {bInterpResult=true}
    boneToModify {boneName=dyn_jacket_front_l_01}
    rotation {yaw=10}
    translation {y=-5}
   }
   n6 = "Transform (Modify) Bone - Bone: dyn_jacket_side_l_01"!(ComponentPose=@n5, bAlphaBoolEnabled:bool=True, Translation="0,-4,0", Rotation="0,10,0", Scale="1,1,1") -> (Pose) {
    alphaInputType=Bool
    rotationMode=BMM_Additive
    rotationSpace=BCS_BoneSpace
    translationMode=BMM_Additive
    translationSpace=BCS_BoneSpace
    alphaBoolBlend {blendInTime=0.3, blendOption=ExpInOut, blendOutTime=0.3}
    alphaScaleBiasClamp {bInterpResult=true}
    boneToModify {boneName=dyn_jacket_side_l_01}
    rotation {yaw=10}
    translation {y=-4}
   }
   n7 = "Transform (Modify) Bone - Bone: dyn_jacket_back_side_r_01"!(ComponentPose=@n6, bAlphaBoolEnabled:bool=True, Translation="3,0,0", Rotation="0,10,0", Scale="1,1,1") -> (Pose) {
    alphaInputType=Bool
    rotationMode=BMM_Additive
    rotationSpace=BCS_BoneSpace
    translationMode=BMM_Additive
    translationSpace=BCS_BoneSpace
    alphaBoolBlend {blendInTime=0.3, blendOption=ExpInOut, blendOutTime=0.3}
    alphaScaleBiasClamp {bInterpResult=true, interpSpeedDecreasing=2, interpSpeedIncreasing=2}
    boneToModify {boneName=dyn_jacket_back_side_r_01}
    rotation {yaw=10}
    translation {x=3}
   }
   n8 = "Transform (Modify) Bone - Bone: dyn_jacket_back_01"!(ComponentPose=@n7, bAlphaBoolEnabled:bool=True, Translation="2,0,0", Rotation="0,20,0", Scale="1,1,1") -> (Pose) {
    alphaInputType=Bool
    rotationMode=BMM_Additive
    rotationSpace=BCS_BoneSpace
    translationMode=BMM_Additive
    translationSpace=BCS_BoneSpace
    alphaBoolBlend {blendInTime=0.3, blendOption=ExpInOut, blendOutTime=0.3}
    alphaScaleBiasClamp {bInterpResult=true, interpSpeedDecreasing=2, interpSpeedIncreasing=2}
    boneToModify {boneName=dyn_jacket_back_01}
    rotation {yaw=20}
    translation {x=2}
   }
   n9 = "Transform (Modify) Bone - Bone: dyn_jacket_back_side_l_01"!(ComponentPose=@n8, bAlphaBoolEnabled:bool=True, Translation="3,0,0", Rotation="0,10,0", Scale="1,1,1") -> (Pose) {
    alphaInputType=Bool
    rotationMode=BMM_Additive
    rotationSpace=BCS_BoneSpace
    translationMode=BMM_Additive
    translationSpace=BCS_BoneSpace
    alphaBoolBlend {blendInTime=0.3, blendOption=ExpInOut, blendOutTime=0.3}
    alphaScaleBiasClamp {bInterpResult=true, interpSpeedDecreasing=2, interpSpeedIncreasing=2}
    boneToModify {boneName=dyn_jacket_back_side_l_01}
    rotation {yaw=10}
    translation {x=3}
   }
   n10 = "Transform (Modify) Bone - Bone: dyn_jacket_back_side_r_01"(ComponentPose=@n9, bAlphaBoolEnabled=True, Translation="0,0,0", Rotation="0,-25,0", Scale="1,1,1") {
    alphaInputType=Bool
    rotationMode=BMM_Additive
    rotationSpace=BCS_BoneSpace
    translationMode=BMM_Additive
    translationSpace=BCS_BoneSpace
    alphaBoolBlend {blendInTime=0.3, blendOption=ExpInOut, blendOutTime=0.3}
    alphaScaleBiasClamp {bInterpResult=true, interpSpeedDecreasing=2, interpSpeedIncreasing=2}
    boneToModify {boneName=dyn_jacket_back_side_r_01}
    rotation {yaw=-25}
   }
   n11 = "Transform (Modify) Bone - Bone: dyn_jacket_back_01"(ComponentPose=@n10, bAlphaBoolEnabled=True, Translation="0,0,0", Rotation="0,-40,0", Scale="1,1,1") {
    alphaInputType=Bool
    rotationMode=BMM_Additive
    rotationSpace=BCS_BoneSpace
    translationMode=BMM_Additive
    translationSpace=BCS_BoneSpace
    alphaBoolBlend {blendInTime=0.3, blendOption=ExpInOut, blendOutTime=0.3}
    alphaScaleBiasClamp {bInterpResult=true, interpSpeedDecreasing=2, interpSpeedIncreasing=2}
    boneToModify {boneName=dyn_jacket_back_01}
    rotation {yaw=-40}
   }
   n12 = "Transform (Modify) Bone - Bone: dyn_jacket_back_side_l_01"(ComponentPose=@n11, bAlphaBoolEnabled=True, Translation="0,0,0", Rotation="0,-25,0", Scale="1,1,1") {
    alphaInputType=Bool
    rotationMode=BMM_Additive
    rotationSpace=BCS_BoneSpace
    translationMode=BMM_Additive
    translationSpace=BCS_BoneSpace
    alphaBoolBlend {blendInTime=0.3, blendOption=ExpInOut, blendOutTime=0.3}
    alphaScaleBiasClamp {bInterpResult=true, interpSpeedDecreasing=2, interpSpeedIncreasing=2}
    boneToModify {boneName=dyn_jacket_back_side_l_01}
    rotation {yaw=-25}
   }
   Output Pose(Result=Component To Local(ComponentPose=@n12)) {layerGroup=None}
   n13 = RigidBody(ComponentPose=Local To Component(LocalPose=Use cached pose 'Pre-rigid body'())) {
    overridePhysicsAsset=&/M_MED_Mummy_Jacket_OnGround_Physics
    simulationSpace=BaseBoneSpace
    alphaBoolBlend {blendOption=Linear}
    baseBoneRef {boneName=spine_05}
    componentAppliedLinearAccClamp=<2000 2000 2000>
    componentLinearAccScale=<0.6 0.6 0.6>
    componentLinearVelScale=<0.6 0.6 0.6>
   }
   Save cached pose 'initial setup'(Pose=Component To Local(ComponentPose=@n13)) {cacheName=initial setup}
   n14 = Control Rig(Source=Copy Pose From Mesh() {bUseAttachedParent=true}) {
    alphaBoolBlend {blendOption=Linear}
    controlRigAssetReference {blueprintRigClass=&/M_MED_Mummy_CtrlRig.M_MED_Mummy_CtrlRig_C}
   }
   Save cached pose 'Pre-rigid body'(Pose=@n14) {cacheName=Pre-rigid body}
   n15 = Get R Arm Raise While Targeting() -> (R Arm Raise While Targeting:float)
   n16 = "Transform (Modify) Bone - Bone: deform_Jacket_r_aim_corrector"(
    ComponentPose=Local To Component(LocalPose=Use cached pose 'Post-Collar'())
    Alpha=@n15
    Translation="15,0,0"
    Rotation="0,0,0"
    Scale="1,1,1"
   ) {
    translationMode=BMM_Additive
    translationSpace=BCS_BoneSpace
    alphaBoolBlend {blendOption=Linear}
    alphaScaleBiasClamp {bInterpResult=true, interpSpeedDecreasing=1, interpSpeedIncreasing=1}
    boneToModify {boneName=deform_Jacket_r_aim_corrector}
    translation {x=15}
   }
   n17 = "Transform (Modify) Bone - Bone: deform_Jacket_l_aim_corrector"(ComponentPose=@n16, Alpha=@n15, Translation="15,0,0", Rotation="0,0,0", Scale="1,1,1") {
    translationMode=BMM_Additive
    translationSpace=BCS_BoneSpace
    alphaBoolBlend {blendOption=Linear}
    alphaScaleBiasClamp {bInterpResult=true, interpSpeedDecreasing=1, interpSpeedIncreasing=1}
    boneToModify {boneName=deform_Jacket_l_aim_corrector}
    translation {x=15}
   }
   Save cached pose 'Post-Correctors'(Pose=Component To Local(ComponentPose=@n17)) {cacheName=Post-Correctors}
   n18 = Get R Arm Raise 45 to 95() -> (R Arm Raise 45 to 95:float)
   n19 = "Transform (Modify) Bone - Bone: dyn_jacket_forward_collar_r_01"(
    ComponentPose=Local To Component(LocalPose=Use cached pose 'initial setup'())
    Alpha=@n18
    Translation="0,4,2"
    Rotation="0,0,0"
    Scale="1,1,1"
   ) {
    rotationMode=BMM_Additive
    rotationSpace=BCS_BoneSpace
    translationMode=BMM_Additive
    translationSpace=BCS_BoneSpace
    alphaBoolBlend {blendOption=Linear}
    alphaScaleBiasClamp {bInterpResult=true}
    boneToModify {boneName=dyn_jacket_forward_collar_r_01}
    translation {y=4, z=2}
   }
   n20 = "Transform (Modify) Bone - Bone: dyn_jacket_collar_side_r_01"(ComponentPose=@n19, Alpha=@n18, Translation="1,0,0", Rotation="0,0,0", Scale="1,1,1") {
    rotationMode=BMM_Additive
    translationMode=BMM_Additive
    alphaBoolBlend {blendOption=Linear}
    alphaScaleBiasClamp {bInterpResult=true}
    boneToModify {boneName=dyn_jacket_collar_side_r_01}
    translation {x=1}
   }
   n21 = "Transform (Modify) Bone - Bone: dyn_jacket_collar_side_r_01"(ComponentPose=@n20, Alpha=Get R Arm Raise 95 to 135() -> (R Arm Raise 95 to 135:float), Translation="1,0,0", Rotation="0,0,0", Scale="1,1,1") {
    rotationMode=BMM_Additive
    translationMode=BMM_Additive
    alphaBoolBlend {blendOption=Linear}
    alphaScaleBiasClamp {bInterpResult=true}
    boneToModify {boneName=dyn_jacket_collar_side_r_01}
    translation {x=1}
   }
   n22 = Get L Arm Raise 60 To 105() -> (L Arm Raise 60 To 105:float)
   n23 = "Transform (Modify) Bone - Bone: dyn_jacket_forward_collar_l_01"(ComponentPose=@n21, Alpha=@n22, Translation="0,0,0", Rotation="10,0,0", Scale="1,1,1") {
    rotationMode=BMM_Additive
    translationMode=BMM_Additive
    alphaBoolBlend {blendOption=Linear}
    alphaScaleBiasClamp {bInterpResult=true}
    boneToModify {boneName=dyn_jacket_forward_collar_l_01}
    rotation {pitch=10}
   }
   n24 = "Transform (Modify) Bone - Bone: dyn_jacket_collar_side_l_01"(ComponentPose=@n23, Alpha=@n22, Translation="0,0,0", Rotation="30,0,0", Scale="1,1,1") {
    rotationMode=BMM_Additive
    translationMode=BMM_Additive
    alphaBoolBlend {blendOption=Linear}
    alphaScaleBiasClamp {bInterpResult=true}
    boneToModify {boneName=dyn_jacket_collar_side_l_01}
    rotation {pitch=30}
   }
   n25 = Get Spine rotation() -> (Spine rotation:float)
   n26 = "Transform (Modify) Bone - Bone: dyn_jacket_forward_collar_r_01"(ComponentPose=@n24, Alpha=@n25, Translation="2,0,0", Rotation="0,0,0", Scale="1,1,1") {
    rotationMode=BMM_Additive
    translationMode=BMM_Additive
    alphaBoolBlend {blendOption=Linear}
    alphaScaleBiasClamp {bInterpResult=true}
    boneToModify {boneName=dyn_jacket_forward_collar_r_01}
    translation {x=2}
   }
   n27 = "Transform (Modify) Bone - Bone: dyn_jacket_collar_side_r_01"(ComponentPose=@n26, Alpha=@n25, Translation="2,0,0", Rotation="0,0,0", Scale="1,1,1") {
    rotationMode=BMM_Additive
    translationMode=BMM_Additive
    alphaBoolBlend {blendOption=Linear}
    alphaScaleBiasClamp {bInterpResult=true}
    boneToModify {boneName=dyn_jacket_collar_side_r_01}
    translation {x=2}
   }
   n28 = "Transform (Modify) Bone - Bone: dyn_jacket_collar_back_01"!(ComponentPose=@n27, bAlphaBoolEnabled:bool=True, Translation="0,-3,0", Rotation="0,-15,0", Scale="0.7,0.7,0.7") -> (Pose) {
    alphaInputType=Bool
    rotationMode=BMM_Additive
    rotationSpace=BCS_BoneSpace
    scaleMode=BMM_Additive
    translationMode=BMM_Additive
    translationSpace=BCS_BoneSpace
    alphaBoolBlend {blendInTime=0.3, blendOption=ExpInOut, blendOutTime=0.3}
    alphaScaleBiasClamp {bInterpResult=true}
    boneToModify {boneName=dyn_jacket_collar_back_01}
    rotation {yaw=-15}
    scale=<0.7 0.7 0.7>
    translation {y=-3}
   }
   n29 = "Transform (Modify) Bone - Bone: dyn_jacket_collar_back_side_r_01"!(ComponentPose=@n28, bAlphaBoolEnabled:bool=True, Translation="0,0,0", Rotation="0,-20,0", Scale="1,1,1") -> (Pose) {
    alphaInputType=Bool
    rotationMode=BMM_Additive
    rotationSpace=BCS_BoneSpace
    scaleMode=BMM_Additive
    translationMode=BMM_Additive
    translationSpace=BCS_BoneSpace
    alphaBoolBlend {blendInTime=0.3, blendOption=ExpInOut, blendOutTime=0.3}
    alphaScaleBiasClamp {bInterpResult=true}
    boneToModify {boneName=dyn_jacket_collar_back_side_r_01}
    rotation {yaw=-20}
   }
   n30 = "Transform (Modify) Bone - Bone: dyn_jacket_collar_back_side_l_01"!(ComponentPose=@n29, bAlphaBoolEnabled:bool=True, Translation="0,0,0", Rotation="0,-20,0", Scale="1,1,1") -> (Pose) {
    alphaInputType=Bool
    rotationMode=BMM_Additive
    rotationSpace=BCS_BoneSpace
    scaleMode=BMM_Additive
    translationMode=BMM_Additive
    translationSpace=BCS_BoneSpace
    alphaBoolBlend {blendInTime=0.3, blendOption=ExpInOut, blendOutTime=0.3}
    alphaScaleBiasClamp {bInterpResult=true}
    boneToModify {boneName=dyn_jacket_collar_back_side_l_01}
    rotation {yaw=-20}
   }
   n31 = "Transform (Modify) Bone - Bone: dyn_jacket_collar_back_01"!(ComponentPose=@n30, Alpha:float=1, Translation="-5,-5,0", Rotation="0,-50,0", Scale="0.4,0.4,0.4") -> (Pose) {
    rotationMode=BMM_Additive
    rotationSpace=BCS_BoneSpace
    scaleMode=BMM_Additive
    translationMode=BMM_Additive
    translationSpace=BCS_BoneSpace
    alphaBoolBlend {blendInTime=0.3, blendOption=ExpInOut, blendOutTime=0.3}
    boneToModify {boneName=dyn_jacket_collar_back_01}
    rotation {yaw=-50}
    scale=<0.4 0.4 0.4>
    translation=<-5 -5>
   }
   n32 = "Transform (Modify) Bone - Bone: dyn_jacket_collar_back_side_r_01"!(ComponentPose=@n31, Alpha:float=1, Translation="-2,-5,0", Rotation="0,-50,0", Scale="0.6,0.6,0.6") -> (Pose) {
    rotationMode=BMM_Additive
    rotationSpace=BCS_BoneSpace
    scaleMode=BMM_Additive
    translationMode=BMM_Additive
    translationSpace=BCS_BoneSpace
    alphaBoolBlend {blendInTime=0.3, blendOption=ExpInOut, blendOutTime=0.3}
    boneToModify {boneName=dyn_jacket_collar_back_side_r_01}
    rotation {yaw=-50}
    scale=<0.6 0.6 0.6>
    translation=<-2 -5>
   }
   n33 = "Transform (Modify) Bone - Bone: dyn_jacket_collar_back_side_l_01"!(ComponentPose=@n32, Alpha:float=1, Translation="-2,-5,0", Rotation="0,-50,0", Scale="0.6,0.6,0.6") -> (Pose) {
    rotationMode=BMM_Additive
    rotationSpace=BCS_BoneSpace
    scaleMode=BMM_Additive
    translationMode=BMM_Additive
    translationSpace=BCS_BoneSpace
    alphaBoolBlend {blendInTime=0.3, blendOption=ExpInOut, blendOutTime=0.3}
    boneToModify {boneName=dyn_jacket_collar_back_side_l_01}
    rotation {yaw=-50}
    scale=<0.6 0.6 0.6>
    translation=<-2 -5>
   }
   n34 = "Transform (Modify) Bone - Bone: dyn_jacket_collar_back_side_r_01"(ComponentPose=@n33, Alpha=1, Translation="-2,-5,0", Rotation="20,-60,0", Scale="0.6,0.6,0.6") {
    rotationMode=BMM_Additive
    rotationSpace=BCS_BoneSpace
    scaleMode=BMM_Additive
    translationMode=BMM_Additive
    translationSpace=BCS_BoneSpace
    alphaBoolBlend {blendInTime=0.3, blendOption=ExpInOut, blendOutTime=0.3}
    alphaScaleBiasClamp {interpSpeedDecreasing=20, interpSpeedIncreasing=20}
    boneToModify {boneName=dyn_jacket_collar_back_side_r_01}
    rotation {pitch=20, yaw=-60}
    scale=<0.6 0.6 0.6>
    translation=<-2 -5>
   }
   n35 = "Transform (Modify) Bone - Bone: dyn_jacket_collar_side_r_01"(ComponentPose=@n34, Alpha=1, Translation="0,0,-1", Rotation="0,0,0", Scale="0.6,0.6,0.6") {
    rotationMode=BMM_Additive
    scaleMode=BMM_Additive
    translationMode=BMM_Additive
    alphaBoolBlend {blendOption=Linear}
    alphaScaleBiasClamp {interpSpeedDecreasing=20, interpSpeedIncreasing=20}
    boneToModify {boneName=dyn_jacket_collar_side_r_01}
    scale=<0.6 0.6 0.6>
    translation {z=-1}
   }
   n36 = "Transform (Modify) Bone - Bone: dyn_jacket_collar_back_side_r_01"(ComponentPose=@n35, Alpha=1, Translation="0,0,-2", Rotation="30,0,0", Scale="0.6,0.6,0.6") {
    rotationMode=BMM_Additive
    scaleMode=BMM_Additive
    translationMode=BMM_Additive
    alphaBoolBlend {blendOption=Linear}
    alphaScaleBiasClamp {interpSpeedDecreasing=20, interpSpeedIncreasing=20}
    boneToModify {boneName=dyn_jacket_collar_back_side_r_01}
    rotation {pitch=30}
    scale=<0.6 0.6 0.6>
    translation {z=-2}
   }
   n37 = "Transform (Modify) Bone - Bone: dyn_jacket_collar_back_01"(ComponentPose=@n36, Alpha=1, Translation="0,0,0", Rotation="0,0,-30", Scale="0.6,0.6,0.6") {
    rotationMode=BMM_Additive
    scaleMode=BMM_Additive
    translationMode=BMM_Additive
    alphaBoolBlend {blendOption=Linear}
    alphaScaleBiasClamp {interpSpeedDecreasing=20, interpSpeedIncreasing=20}
    boneToModify {boneName=dyn_jacket_collar_back_01}
    rotation {roll=-30}
    scale=<0.6 0.6 0.6>
   }
   n38 = "Transform (Modify) Bone - Bone: dyn_jacket_collar_back_01"!(ComponentPose=@n37, bAlphaBoolEnabled:bool=True, Translation="0,0,0", Rotation="0,-40,0", Scale="0.4,0.4,0.4") -> (Pose) {
    alphaInputType=Bool
    rotationMode=BMM_Additive
    rotationSpace=BCS_BoneSpace
    scaleMode=BMM_Additive
    translationMode=BMM_Additive
    translationSpace=BCS_BoneSpace
    alphaBoolBlend {blendInTime=0.3, blendOption=ExpInOut, blendOutTime=0.3}
    alphaScaleBiasClamp {bInterpResult=true}
    boneToModify {boneName=dyn_jacket_collar_back_01}
    rotation {yaw=-40}
    scale=<0.4 0.4 0.4>
   }
   n39 = "Transform (Modify) Bone - Bone: dyn_jacket_collar_back_side_l_01"!(ComponentPose=@n38, bAlphaBoolEnabled:bool=True, Translation="0,0,0", Rotation="0,-55,0", Scale="1,1,1") -> (Pose) {
    alphaInputType=Bool
    rotationMode=BMM_Additive
    rotationSpace=BCS_BoneSpace
    scaleMode=BMM_Additive
    translationMode=BMM_Additive
    translationSpace=BCS_BoneSpace
    alphaBoolBlend {blendInTime=0.3, blendOption=ExpInOut, blendOutTime=0.3}
    alphaScaleBiasClamp {bInterpResult=true}
    boneToModify {boneName=dyn_jacket_collar_back_side_l_01}
    rotation {yaw=-55}
   }
   Save cached pose 'Post-Collar'(Pose=Component To Local(ComponentPose=@n39)) {cacheName=Post-Collar}
   n40 = Anim Dynamics - Bone: dyn_left_leg_pouch_01(
    ComponentPose=Local To Component(LocalPose=Use cached pose 'Post-Correctors'())
    Alpha=Get Is Not Crouching Alpha() -> (Is Not Crouching Alpha:double)
    ExternalForce="0,0,0"
   ) {
    angularBiasOverride=1
    angularSpringConstant=10000
    bAngularSpring=true
    bOverrideAngularBias=true
    simulationSpace=RootRelative
    alphaBoolBlend {blendOption=Linear}
    alphaScaleBiasClamp {bInterpResult=true}
    boundBone {boneName=dyn_left_leg_pouch_01}
    componentAppliedLinearAccClamp=<450 450 450>
    componentLinearAccScale=<0.05 0.05 0.05>
    componentLinearVelScale=<0.05 0.05 0.05>
    physicsBodyDefinitions [{
      boundBone {boneName=dyn_left_leg_pouch_01}
      boxExtents=<10 10 10>
      constraintSetup {angularLimitsMax=<0 0 0>, angularLimitsMin=<0 -30 0>, angularTarget=<0 0 0>}
      localJointOffset=<10 0 0>
     }]
    relativeSpaceBone {boneName=spine_01}
   }
   n41 = "Transform (Modify) Bone - Bone: dyn_left_leg_pouch_01"(ComponentPose=@n40, Alpha=Get PouchTransformAlpha() -> (PouchTransformAlpha:double), Translation="0,0,0", Rotation="20,0,0", Scale="1,1,1") {
    rotationMode=BMM_Additive
    rotationSpace=BCS_BoneSpace
    translationMode=BMM_Additive
    translationSpace=BCS_BoneSpace
    alphaBoolBlend {blendOption=Linear}
    alphaScaleBiasClamp {bInterpResult=true, interpSpeedDecreasing=4, interpSpeedIncreasing=4}
    boneToModify {boneName=dyn_left_leg_pouch_01}
    rotation {pitch=20}
   }
   n42 = Anim Dynamics - Bone: dyn_jacket_forward_collar_r_01(ComponentPose=@n41, Alpha=1, ExternalForce="0,0,0") {
    angularBiasOverride=0.9999
    angularSpringConstant=3000
    bAngularSpring=true
    bOverrideAngularBias=true
    bUseGravityOverride=true
    simulationSpace=RootRelative
    alphaBoolBlend {blendOption=Linear}
    boundBone {boneName=dyn_jacket_forward_collar_r_01}
    componentAppliedLinearAccClamp=<1000 1000 1000>
    componentLinearAccScale=<0.1 0.1 0.1>
    componentLinearVelScale=<0.1 0.1 0.1>
    physicsBodyDefinitions [{
      boundBone {boneName=dyn_jacket_forward_collar_r_01}
      boxExtents=<30 20 20>
      constraintSetup {angularLimitsMax=<6 5 0>, angularLimitsMin=<-25 -5 0>, angularTarget=<0 0 0>}
      localJointOffset=<0 0 15>
     }]
   }
   n43 = Anim Dynamics - Bone: dyn_jacket_forward_collar_l_01(ComponentPose=@n42, Alpha=1, ExternalForce="0,0,0") {
    angularBiasOverride=0.9999
    angularSpringConstant=3000
    bAngularSpring=true
    bOverrideAngularBias=true
    bUseGravityOverride=true
    simulationSpace=RootRelative
    alphaBoolBlend {blendOption=Linear}
    boundBone {boneName=dyn_jacket_forward_collar_l_01}
    componentAppliedLinearAccClamp=<1000 1000 1000>
    componentLinearAccScale=<0.1 0.1 0.1>
    componentLinearVelScale=<0.1 0.1 0.1>
    physicsBodyDefinitions [{
      boundBone {boneName=dyn_jacket_forward_collar_l_01}
      boxExtents=<30 20 20>
      constraintSetup {angularLimitsMax=<25 5 0>, angularLimitsMin=<-6 -5 0>, angularTarget=<0 0 0>}
      localJointOffset=<0 0 -15>
     }]
   }
   n44 = Anim Dynamics - Bone: dyn_jacket_forward_pecs_l_01(ComponentPose=@n43, Alpha=1, ExternalForce="0,0,0") {
    angularBiasOverride=0.9999
    angularSpringConstant=3000
    bAngularSpring=true
    bOverrideAngularBias=true
    bUseGravityOverride=true
    simulationSpace=RootRelative
    alphaBoolBlend {blendOption=Linear}
    boundBone {boneName=dyn_jacket_forward_pecs_l_01}
    componentAppliedLinearAccClamp=<1000 1000 1000>
    componentLinearAccScale=<0.1 0.1 0.1>
    componentLinearVelScale=<0.1 0.1 0.1>
    physicsBodyDefinitions [{
      boundBone {boneName=dyn_jacket_forward_pecs_l_01}
      boxExtents=<30 20 20>
      constraintSetup {angularLimitsMax=<6 5 0>, angularLimitsMin=<-25 -5 0>, angularTarget=<0 0 0>}
      localJointOffset=<0 0 15>
     }]
   }
   n45 = Anim Dynamics - Bone: dyn_jacket_forward_pecs_r_01(ComponentPose=@n44, Alpha=1, ExternalForce="0,0,0") {
    angularBiasOverride=0.9999
    angularSpringConstant=3000
    bAngularSpring=true
    bOverrideAngularBias=true
    bUseGravityOverride=true
    simulationSpace=RootRelative
    alphaBoolBlend {blendOption=Linear}
    boundBone {boneName=dyn_jacket_forward_pecs_r_01}
    componentAppliedLinearAccClamp=<2000 2000 2000>
    componentLinearAccScale=<0.1 0.1 0.1>
    componentLinearVelScale=<0.1 0.1 0.1>
    physicsBodyDefinitions [{
      boundBone {boneName=dyn_jacket_forward_pecs_r_01}
      boxExtents=<30 20 20>
      constraintSetup {angularLimitsMax=<25 5 0>, angularLimitsMin=<-6 -5 0>, angularTarget=<0 0 0>}
      localJointOffset=<0 0 -15>
     }]
   }
   n46 = "Constraint - dyn_zipper_MSTR (P)"(ComponentPose=@n45, Alpha=1, ConstraintWeights_0=1) {
    alphaBoolBlend {blendOption=Linear}
    boneToModify {boneName=dyn_zipper_MSTR}
    constraintSetup [{targetBone {boneName=dyn_jacket_front_r_01}}]
    constraintWeights [1]
   }
   n47 = Anim Dynamics - Bone: dyn_zipper_01(ComponentPose=@n46, Alpha=1, ExternalForce="0,0,0") {
    angularBiasOverride=1
    angularDampingOverride=1
    angularSpringConstant=10000
    bAngularSpring=true
    bOverrideAngularBias=true
    bUseGravityOverride=true
    gravityScale=0.35
    simulationSpace=BoneRelative
    alphaBoolBlend {blendOption=Linear}
    boundBone {boneName=dyn_zipper_01}
    componentAppliedLinearAccClamp=<2000 2000 2000>
    componentLinearAccScale=<0.1 0.1 0.1>
    componentLinearVelScale=<0.2 0.2 0.2>
    gravityOverride {z=-550}
    physicsBodyDefinitions [{
      boundBone {boneName=dyn_zipper_01}
      boxExtents=<10 10 10>
      constraintSetup {angularLimitsMax=<0 0 0>, angularLimitsMin=<0 0 -80>, angularTarget=<0 0 0>}
      localJointOffset=<5 0 0>
     }]
    relativeSpaceBone {boneName=root}
   }
   n48 = Anim Dynamics - Bone: dyn_arm_l_knot_strap_01(ComponentPose=@n47, Alpha=0.5, ExternalForce="0,0,0") {
    angularBiasOverride=1
    angularSpringConstant=1000000
    bAngularSpring=true
    bOverrideAngularBias=true
    bUseGravityOverride=true
    simulationSpace=RootRelative
    alphaBoolBlend {blendOption=Linear}
    boundBone {boneName=dyn_arm_l_knot_strap_01}
    componentAppliedLinearAccClamp=<2000 2000 2000>
    componentLinearAccScale=<0.5 0.5 0.5>
    componentLinearVelScale=<0.5 0.5 0.5>
    gravityOverride {z=-300}
    physicsBodyDefinitions [{
      boundBone {boneName=dyn_arm_l_knot_strap_01}
      boxExtents=<10 30 30>
      constraintSetup {angularLimitsMax=<5 50 50>, angularLimitsMin=<-5 -10 -50>, angularTarget=<0 0 0>}
      localJointOffset=<10 0 0>
     }]
    relativeSpaceBone {boneName=spine_04}
   }
   n49 = Anim Dynamics - Bone: dyn_arm_r_knot_strap_01(ComponentPose=@n48, Alpha=0.5, ExternalForce="0,0,0") {
    angularBiasOverride=1
    angularSpringConstant=1000000
    bAngularSpring=true
    bOverrideAngularBias=true
    bUseGravityOverride=true
    simulationSpace=RootRelative
    alphaBoolBlend {blendOption=Linear}
    boundBone {boneName=dyn_arm_r_knot_strap_01}
    componentAppliedLinearAccClamp=<2000 2000 2000>
    componentLinearAccScale=<0.5 0.5 0.5>
    componentLinearVelScale=<0.5 0.5 0.5>
    gravityOverride {z=-300}
    physicsBodyDefinitions [{
      boundBone {boneName=dyn_arm_r_knot_strap_01}
      boxExtents=<30 30 10>
      constraintSetup {angularLimitsMax=<50 50 5>, angularLimitsMin=<-50 -10 -5>, angularTarget=<0 0 0>}
      localJointOffset=<-10 0 0>
     }]
    relativeSpaceBone {boneName=spine_04}
   }
   n50 = "Transform (Modify) Bone - Bone: dyn_jacket_forward_pecs_l_01"(ComponentPose=@n49, Alpha=1, Translation="0,-3,-2", Rotation="0,0,30", Scale="1,1,1") {
    rotationMode=BMM_Additive
    rotationSpace=BCS_BoneSpace
    translationMode=BMM_Additive
    translationSpace=BCS_BoneSpace
    alphaBoolBlend {blendOption=Linear}
    boneToModify {boneName=dyn_jacket_forward_pecs_l_01}
    rotation {roll=30}
    translation {y=-3, z=-2}
   }
   n51 = "Transform (Modify) Bone - Bone: dyn_jacket_front_r_01"(ComponentPose=@n50, Alpha=1, Translation="0,0,0", Rotation="0,-30,0", Scale="1,1,1") {
    rotationMode=BMM_Additive
    rotationSpace=BCS_BoneSpace
    translationMode=BMM_Additive
    translationSpace=BCS_BoneSpace
    alphaBoolBlend {blendOption=Linear}
    alphaScaleBiasClamp {bInterpResult=true}
    boneToModify {boneName=dyn_jacket_front_r_01}
    rotation {yaw=-30}
   }
   n52 = "Transform (Modify) Bone - Bone: dyn_jacket_front_l_01"!(ComponentPose=@n51, Alpha:float=1, Translation="0,0,0", Rotation="0,-30,0", Scale="1,1,1") -> (Pose) {
    rotationMode=BMM_Additive
    rotationSpace=BCS_BoneSpace
    translationMode=BMM_Additive
    translationSpace=BCS_BoneSpace
    alphaBoolBlend {blendOption=Linear}
    alphaScaleBiasClamp {bInterpResult=true}
    boneToModify {boneName=dyn_jacket_front_l_01}
    rotation {yaw=-30}
   }
   n53 = Anim Dynamics - Bone: dyn_jacket_side_r_01(ComponentPose=@n52, Alpha=0.6, ExternalForce="0,0,0") {
    angularBiasOverride=0.999
    angularDampingOverride=1
    angularSpringConstant=10000
    bOverrideAngularBias=true
    bOverrideAngularDamping=true
    bUseGravityOverride=true
    alphaBoolBlend {blendOption=Linear}
    boundBone {boneName=dyn_jacket_side_r_01}
    componentAppliedLinearAccClamp=<2000 2000 2000>
    componentLinearAccScale=<0.3 0.3 0.3>
    componentLinearVelScale=<0.3 0.3 0.3>
    gravityOverride {z=-980}
    physicsBodyDefinitions [{
      boundBone {boneName=dyn_jacket_side_r_01}
      boxExtents=<20 20 20>
      constraintSetup {angularLimitsMax=<0 2 20>, angularLimitsMin=<0 -2 -40>, angularTarget=<0 0 0>}
      localJointOffset=<10 0 0>
     }]
    relativeSpaceBone {boneName=spine_03}
   }
   n54 = Anim Dynamics - Bone: dyn_jacket_back_side_r_01(ComponentPose=@n53, Alpha=0.45, ExternalForce="0,0,0") {
    angularBiasOverride=0.9999
    angularDampingOverride=1
    angularSpringConstant=1000
    bOverrideAngularBias=true
    bOverrideAngularDamping=true
    bUseGravityOverride=true
    alphaBoolBlend {blendOption=Linear}
    boundBone {boneName=dyn_jacket_back_side_r_01}
    componentAppliedLinearAccClamp=<2000 2000 2000>
    componentLinearAccScale=<0.3 0.3 0.3>
    componentLinearVelScale=<0.3 0.3 0.3>
    gravityOverride {z=-980}
    physicsBodyDefinitions [{
      boundBone {boneName=dyn_jacket_back_side_r_01}
      boxExtents=<20 20 20>
      constraintSetup {angularLimitsMax=<0 2 20>, angularLimitsMin=<0 -2 -40>, angularTarget=<0 0 0>}
      localJointOffset=<10 0 0>
     }]
    relativeSpaceBone {boneName=spine_03}
   }
   n55 = Anim Dynamics - Bone: dyn_jacket_back_01(ComponentPose=@n54, Alpha=0.45, ExternalForce="0,0,0") {
    angularBiasOverride=0.9999
    angularDampingOverride=1
    angularSpringConstant=1000
    bOverrideAngularBias=true
    bOverrideAngularDamping=true
    bUseGravityOverride=true
    alphaBoolBlend {blendOption=Linear}
    boundBone {boneName=dyn_jacket_back_01}
    componentAppliedLinearAccClamp=<2000 2000 2000>
    componentLinearAccScale=<0.3 0.3 0.3>
    componentLinearVelScale=<0.3 0.3 0.3>
    gravityOverride {z=-980}
    physicsBodyDefinitions [{
      boundBone {boneName=dyn_jacket_back_01}
      boxExtents=<20 20 20>
      constraintSetup {angularLimitsMax=<0 2 20>, angularLimitsMin=<0 -2 -40>, angularTarget=<0 0 0>}
      localJointOffset=<10 0 0>
     }]
    relativeSpaceBone {boneName=spine_03}
   }
   n56 = Anim Dynamics - Bone: dyn_jacket_back_side_l_01(ComponentPose=@n55, Alpha=0.45, ExternalForce="0,0,0") {
    angularBiasOverride=0.9999
    angularDampingOverride=1
    angularSpringConstant=1000
    bOverrideAngularBias=true
    bOverrideAngularDamping=true
    bUseGravityOverride=true
    alphaBoolBlend {blendOption=Linear}
    boundBone {boneName=dyn_jacket_back_side_l_01}
    componentAppliedLinearAccClamp=<2000 2000 2000>
    componentLinearAccScale=<0.3 0.3 0.3>
    componentLinearVelScale=<0.3 0.3 0.3>
    gravityOverride {z=-980}
    physicsBodyDefinitions [{
      boundBone {boneName=dyn_jacket_back_side_l_01}
      boxExtents=<20 20 20>
      constraintSetup {angularLimitsMax=<0 2 20>, angularLimitsMin=<0 -2 -40>, angularTarget=<0 0 0>}
      localJointOffset=<10 0 0>
     }]
    relativeSpaceBone {boneName=spine_03}
   }
   n57 = Anim Dynamics - Bone: dyn_jacket_side_l_01(ComponentPose=@n56, Alpha=0.6, ExternalForce="0,0,0") {
    angularBiasOverride=0.9999
    angularDampingOverride=1
    angularSpringConstant=10000
    bOverrideAngularBias=true
    bOverrideAngularDamping=true
    bUseGravityOverride=true
    alphaBoolBlend {blendOption=Linear}
    boundBone {boneName=dyn_jacket_side_l_01}
    componentAppliedLinearAccClamp=<2000 2000 2000>
    componentLinearAccScale=<0.3 0.3 0.3>
    componentLinearVelScale=<0.3 0.3 0.3>
    gravityOverride {z=-980}
    physicsBodyDefinitions [{
      boundBone {boneName=dyn_jacket_side_l_01}
      boxExtents=<20 20 20>
      constraintSetup {angularLimitsMax=<0 2 20>, angularLimitsMin=<0 -2 -40>, angularTarget=<0 0 0>}
      localJointOffset=<10 0 0>
     }]
    relativeSpaceBone {boneName=spine_03}
   }
   n58 = Anim Dynamics - Bone: dyn_jacket_collar_side_r_01(ComponentPose=@n57, Alpha=0.8, ExternalForce="0,0,0") {
    angularBiasOverride=0.999
    angularDampingOverride=0.99
    angularSpringConstant=10000
    bOverrideAngularBias=true
    bOverrideAngularDamping=true
    bUseGravityOverride=true
    simulationSpace=BoneRelative
    alphaBoolBlend {blendOption=Linear}
    boundBone {boneName=dyn_jacket_collar_side_r_01}
    componentAppliedLinearAccClamp=<1000 1000 1000>
    componentLinearAccScale=<0.2 0.2 0.2>
    componentLinearVelScale=<0.2 0.2 0.2>
    gravityOverride {z=400}
    physicsBodyDefinitions [{
      boundBone {boneName=dyn_jacket_collar_side_r_01}
      boxExtents=<10 15 15>
      constraintSetup {angularLimitsMax=<0 5 -5>, angularLimitsMin=<0 -5 -20>, angularTarget=<0 0 0>}
      localJointOffset=<12 0 0>
     }]
    relativeSpaceBone {boneName=spine_05}
   }
   n59 = Anim Dynamics - Bone: dyn_jacket_collar_back_side_r_01(ComponentPose=@n58, Alpha=0.6, ExternalForce="0,0,0") {
    angularBiasOverride=0.999
    angularDampingOverride=0.99
    angularSpringConstant=10000
    bOverrideAngularBias=true
    bOverrideAngularDamping=true
    bUseGravityOverride=true
    simulationSpace=BoneRelative
    alphaBoolBlend {blendOption=Linear}
    boundBone {boneName=dyn_jacket_collar_back_side_r_01}
    componentAppliedLinearAccClamp=<1000 1000 1000>
    componentLinearAccScale=<0.2 0.2 0.2>
    componentLinearVelScale=<0.2 0.2 0.2>
    gravityOverride {z=400}
    physicsBodyDefinitions [{
      boundBone {boneName=dyn_jacket_collar_back_side_r_01}
      boxExtents=<10 15 15>
      constraintSetup {angularLimitsMax=<0 5 -5>, angularLimitsMin=<0 -5 -40>, angularTarget=<0 0 0>}
      localJointOffset=<12 0 0>
     }]
    relativeSpaceBone {boneName=spine_05}
   }
   n60 = Anim Dynamics - Bone: dyn_jacket_collar_back_01(ComponentPose=@n59, Alpha=0.6, ExternalForce="0,0,0") {
    angularBiasOverride=0.999
    angularDampingOverride=0.99
    angularSpringConstant=10000
    bOverrideAngularBias=true
    bOverrideAngularDamping=true
    bUseGravityOverride=true
    simulationSpace=BoneRelative
    alphaBoolBlend {blendOption=Linear}
    boundBone {boneName=dyn_jacket_collar_back_01}
    componentAppliedLinearAccClamp=<1000 1000 1000>
    componentLinearAccScale=<0.2 0.2 0.2>
    componentLinearVelScale=<0.2 0.2 0.2>
    gravityOverride {z=400}
    physicsBodyDefinitions [{
      boundBone {boneName=dyn_jacket_collar_back_01}
      boxExtents=<10 15 15>
      constraintSetup {angularLimitsMax=<0 5 -5>, angularLimitsMin=<0 -5 -40>, angularTarget=<0 0 0>}
      localJointOffset=<12 0 0>
     }]
    relativeSpaceBone {boneName=spine_05}
   }
   n61 = Anim Dynamics - Bone: dyn_jacket_collar_back_side_l_01(ComponentPose=@n60, Alpha=0.6, ExternalForce="0,0,0") {
    angularBiasOverride=0.999
    angularDampingOverride=0.99
    angularSpringConstant=10000
    bOverrideAngularBias=true
    bOverrideAngularDamping=true
    bUseGravityOverride=true
    simulationSpace=BoneRelative
    alphaBoolBlend {blendOption=Linear}
    boundBone {boneName=dyn_jacket_collar_back_side_l_01}
    componentAppliedLinearAccClamp=<1000 1000 1000>
    componentLinearAccScale=<0.2 0.2 0.2>
    componentLinearVelScale=<0.2 0.2 0.2>
    gravityOverride {z=400}
    physicsBodyDefinitions [{
      boundBone {boneName=dyn_jacket_collar_back_side_l_01}
      boxExtents=<10 15 15>
      constraintSetup {angularLimitsMax=<0 5 -5>, angularLimitsMin=<0 -5 -40>, angularTarget=<0 0 0>}
      localJointOffset=<12 0 0>
     }]
    relativeSpaceBone {boneName=spine_05}
   }
   n62 = Anim Dynamics - Bone: dyn_jacket_collar_side_l_01(ComponentPose=@n61, Alpha=0.8, ExternalForce="0,0,0") {
    angularBiasOverride=0.999
    angularDampingOverride=0.99
    angularSpringConstant=10000
    bOverrideAngularBias=true
    bOverrideAngularDamping=true
    bUseGravityOverride=true
    simulationSpace=BoneRelative
    alphaBoolBlend {blendOption=Linear}
    boundBone {boneName=dyn_jacket_collar_side_l_01}
    componentAppliedLinearAccClamp=<1000 1000 1000>
    componentLinearAccScale=<0.2 0.2 0.2>
    componentLinearVelScale=<0.2 0.2 0.2>
    gravityOverride {z=400}
    physicsBodyDefinitions [{
      boundBone {boneName=dyn_jacket_collar_side_l_01}
      boxExtents=<10 15 15>
      constraintSetup {angularLimitsMax=<0 5 -5>, angularLimitsMin=<0 -5 -20>, angularTarget=<0 0 0>}
      localJointOffset=<12 0 0>
     }]
    relativeSpaceBone {boneName=spine_05}
   }
   n63 = Anim Dynamics - Bone: dyn_jacket_forward_collar_l_01(ComponentPose=@n62, Alpha=0.8, ExternalForce="0,0,0") {
    angularBiasOverride=0.999
    angularDampingOverride=0.99
    angularSpringConstant=10000
    bOverrideAngularBias=true
    bOverrideAngularDamping=true
    bUseGravityOverride=true
    simulationSpace=BoneRelative
    alphaBoolBlend {blendOption=Linear}
    boundBone {boneName=dyn_jacket_forward_collar_l_01}
    componentAppliedLinearAccClamp=<1000 1000 1000>
    componentLinearAccScale=<0.2 0.2 0.2>
    componentLinearVelScale=<0.2 0.2 0.2>
    gravityOverride {z=400}
    physicsBodyDefinitions [{
      boundBone {boneName=dyn_jacket_forward_collar_l_01}
      boxExtents=<10 15 15>
      constraintSetup {angularLimitsMax=<0 5 -5>, angularLimitsMin=<0 -5 -20>, angularTarget=<0 0 0>}
      localJointOffset=<12 0 0>
     }]
    relativeSpaceBone {boneName=spine_05}
   }
   n64 = Anim Dynamics - Bone: dyn_jacket_forward_collar_r_01(ComponentPose=@n63, Alpha=0.8, ExternalForce="0,0,0") {
    angularBiasOverride=0.999
    angularDampingOverride=0.99
    angularSpringConstant=10000
    bOverrideAngularBias=true
    bOverrideAngularDamping=true
    bUseGravityOverride=true
    simulationSpace=BoneRelative
    alphaBoolBlend {blendOption=Linear}
    boundBone {boneName=dyn_jacket_forward_collar_r_01}
    componentAppliedLinearAccClamp=<1000 1000 1000>
    componentLinearAccScale=<0.2 0.2 0.2>
    componentLinearVelScale=<0.2 0.2 0.2>
    gravityOverride {z=400}
    physicsBodyDefinitions [{
      boundBone {boneName=dyn_jacket_forward_collar_r_01}
      boxExtents=<10 15 15>
      constraintSetup {angularLimitsMax=<0 5 -5>, angularLimitsMin=<0 -5 -20>, angularTarget=<0 0 0>}
      localJointOffset=<12 0 0>
     }]
    relativeSpaceBone {boneName=spine_05}
   }
   n65 = "Transform (Modify) Bone - Bone: dyn_jacket_collar_back_01"!(ComponentPose=@n64, Alpha:float=1, Translation="0,-6,0", Rotation="0,-15,0", Scale="1,1,1") -> (Pose) {
    rotationMode=BMM_Additive
    rotationSpace=BCS_BoneSpace
    translationMode=BMM_Additive
    translationSpace=BCS_BoneSpace
    alphaBoolBlend {blendOption=Linear}
    alphaScaleBiasClamp {bInterpResult=true, interpSpeedDecreasing=5, interpSpeedIncreasing=5}
    boneToModify {boneName=dyn_jacket_collar_back_01}
    rotation {yaw=-15}
    translation {y=-6}
   }
   n66 = "Transform (Modify) Bone - Bone: dyn_jacket_collar_back_side_r_01"(ComponentPose=@n65, Alpha=1, Translation="0,-4,0", Rotation="0,-10,0", Scale="1,1,1") {
    rotationMode=BMM_Additive
    rotationSpace=BCS_BoneSpace
    translationMode=BMM_Additive
    translationSpace=BCS_BoneSpace
    alphaBoolBlend {blendOption=Linear}
    alphaScaleBiasClamp {bInterpResult=true, interpSpeedDecreasing=5, interpSpeedIncreasing=5}
    boneToModify {boneName=dyn_jacket_collar_back_side_r_01}
    rotation {yaw=-10}
    translation {y=-4}
   }
   n67 = "Transform (Modify) Bone - Bone: dyn_jacket_collar_back_side_l_01"!(ComponentPose=@n66, Alpha:float=1, Translation="0,-4,0", Rotation="0,-10,0", Scale="1,1,1") -> (Pose) {
    rotationMode=BMM_Additive
    rotationSpace=BCS_BoneSpace
    translationMode=BMM_Additive
    translationSpace=BCS_BoneSpace
    alphaBoolBlend {blendOption=Linear}
    alphaScaleBiasClamp {bInterpResult=true, interpSpeedDecreasing=5, interpSpeedIncreasing=5}
    boneToModify {boneName=dyn_jacket_collar_back_side_l_01}
    rotation {yaw=-10}
    translation {y=-4}
   }
   Save cached pose 'Post Anim Dynamics'(Pose=Component To Local(ComponentPose=@n67)) {cacheName=Post Anim Dynamics}
  ]}
 EventGraph {nodes [
   n1 = Event BlueprintUpdateAnimation()
   > Set Delta time X(Delta time X:double=@n1.DeltaTimeX) -> (Output_Get:double)
   > n2 = Sequence!(execute) -> (then_0, then_1, then_2:Exec, then_3:Exec, then_4:Exec, then_5:Exec, then_6:Exec, then_7:Exec)
   > Set Jumping or Falling Check(Jumping or Falling Check:bool=OR Boolean(A=Get bIsJumping(), B=Get bIsFalling())) -> (Output_Get:bool)
   n3 = Get bIsCrouching()
   > Set Is Not Crouching Alpha(Is Not Crouching Alpha:double=SelectFloat(A=0.6, B=0, bPickA=NOT Boolean(A=@n3))) -> (Output_Get:double)
   n4 = OR Boolean(A=Get bIsJumping(), B=Get bIsSliding())
   > n5 = Set PouchTransformAlpha(PouchTransformAlpha:double=SelectFloat(A=1, B=0, bPickA=@n4)) -> (Output_Get:double)
   > n6 = Branch(Condition=@n4)
   > Set PouchTransformAlpha(PouchTransformAlpha=FInterpTo(Current=@n5, Target=1, DeltaTime=Get Delta Time X() -> (Delta Time X:double), InterpSpeed=10))
   Set PouchTransformAlpha(execute=@n6.else, PouchTransformAlpha=0)
   n7 = Get Body Mesh() -> (Body Mesh:Skeletal Mesh Component Object Reference)
   n8 = Get bIsTargeting()
   Set R Arm Raise While Targeting(
    execute=@n2.then_1
    R Arm Raise While Targeting:float=float * float(
     A="To Float (Boolean)"(InBool=@n8)
     B=GetAlphaFromReferenceTransformDelta(SkeletalMeshComponent=@n7, BoneName=upperarm_r, BaseName=spine_05, InputRange=(X=45,Y=135), TwistAxis="0,0,1")
    ) {is=title K2Node_PromotableOperator}
   ) -> (Output_Get:float)
   > Set R Arm Raise 45 to 95(R Arm Raise 45 to 95:float=GetAlphaFromReferenceTransformDelta(SkeletalMeshComponent=@n7, BoneName=upperarm_r, BaseName=spine_05, InputRange=(X=45,Y=95), TwistAxis="0,0,1")) -> (Output_Get:float)
   > Set R Arm Raise 95 to 135(R Arm Raise 95 to 135:float=GetAlphaFromReferenceTransformDelta(SkeletalMeshComponent=@n7, BoneName=upperarm_r, BaseName=spine_05, InputRange=(X=95,Y=135), TwistAxis="0,0,1")) -> (Output_Get:float)
   n9 = Get bIsTargeting()
   > Set Spine rotation(Spine rotation:float=float * float(
     A="To Float (Boolean)"(InBool=@n9)
     B=GetAlphaFromReferenceTransformDelta(SkeletalMeshComponent=@n7, BoneName=spine_05, BaseName=pelvis, DeltaComponentType=EulerX, InputRange=(X=-45,Y=5), TwistAxis="0,0,1")
    )) -> (Output_Get:float)
   > Set LeftLegBendEulerY_Alpha(LeftLegBendEulerY_Alpha:double=GetAlphaFromReferenceTransformDelta(SkeletalMeshComponent=@n7, BoneName=thigh_l, BaseName=pelvis, DeltaComponentType=EulerY, InputRange=(X=10,Y=30), TwistAxis="0,0,1")) -> (Output_Get:double)
   > Set L Arm Raise 60 To 105(L Arm Raise 60 To 105:float=GetAlphaFromReferenceTransformDelta(SkeletalMeshComponent=@n7, BoneName=upperarm_l, BaseName=spine_05, InputRange=(X=60,Y=105), TwistAxis="0,0,1")) -> (Output_Get:float)
   > "Set L Arm Pivot Euler Y 0 To -16 "("L Arm Pivot Euler Y 0 To -16 ":float=GetAlphaFromReferenceTransformDelta(SkeletalMeshComponent=@n7, BoneName=upperarm_l, BaseName=spine_05, DeltaComponentType=EulerY, InputRange=(X=0,Y=-16), TwistAxis="0,0,1")) -> (Output_Get:float)
   n10 = Get Jumping or Falling Check() -> (Jumping or Falling Check:bool)
   n11 = Branch(execute=@n2.then_6, Condition=@n10)
   n12 = TryGetPawnOwner()
   n13 = GetVelocity!(self:Actor Object Reference=@n12) -> (ReturnValue:Vector)
   n14 = Dot Product!(A:Vector=@n13, B:Vector=GetActorRightVector!(self:Actor Object Reference=@n12) -> (ReturnValue:Vector)) -> (ReturnValue:double)
   n15 = Set Jumping to the Left Alpha(execute=@n19.then, Jumping to the Left Alpha:double=MapRangeClamped(Value=@n14, InRangeA=0, InRangeB=-230)) -> (Output_Get:double)
   Set Jumping to the Left Alpha(execute=@n11.then, Jumping to the Left Alpha=@n15)
   Set Jumping to the Left Alpha(execute=@n11.else, Jumping to the Left Alpha=0)
   n16 = Branch(execute=@n2.then_7, Condition=@n10)
   n17 = Set Jumping to the Right Alpha(execute=@n15.then, Jumping to the Right Alpha:double=MapRangeClamped(Value=@n14, InRangeA=0, InRangeB=230)) -> (Output_Get:double)
   Set Jumping to the Right Alpha(execute=@n16.then, Jumping to the Right Alpha=@n17)
   Set Jumping to the Right Alpha(execute=@n16.else, Jumping to the Right Alpha=0)
   Set HeadLookUpAlpha(
    execute=@n2.then_2
    HeadLookUpAlpha:double=GetAlphaFromReferenceTransformDelta(SkeletalMeshComponent=@n7, BoneName=head, BaseName=spine_05, InputRange=(X=-25,Y=-65), OutputRange=(X=0,Y=0.6), TwistAxis="0,0,1")
   ) -> (Output_Get:double)
   > Set Head Tilt To Right Alpha(Head Tilt To Right Alpha:double=GetAlphaFromReferenceTransformDelta(
     SkeletalMeshComponent=@n7
     BoneName=head
     BaseName=spine_05
     DeltaComponentType=EulerY
     InputRange=(X=0,Y=30)
     OutputRange=(X=0,Y=0.7)
     TwistAxis="0,0,1"
    )) -> (Output_Get:double)
   n18 = Branch(execute=@n2.then_3, Condition=Get bIsCrouching())
   > Set Crouching RArmRaise(Crouching RArmRaise:double=Get R Arm Raise 45 to 95()) -> (Output_Get:double)
   Set Crouching RArmRaise(execute=@n18.else, Crouching RArmRaise=0)
   n19 = Set Crouching to the left Alpha(execute=@n2.then_4, Crouching to the left Alpha:double=MapRangeClamped(Value=@n14, InRangeA=0, InRangeB=-230)) -> (Output_Get:double)
   n20 = Dot Product(A=@n13, B=GetActorForwardVector!(self:Actor Object Reference=@n12) -> (ReturnValue:Vector))
   Set Moving Backwards Alpha(execute=@n2.then_5, Moving Backwards Alpha:double=MapRangeClamped(Value=@n20, InRangeA=0, InRangeB=-300)) -> (Output_Get:double)
   > n21 = "Set Crouching forward of  backwards Alpha"("Crouching forward of  backwards Alpha":double=MapRangeClamped(Value="Absolute (Float)"(A=@n20), InRangeA=0, InRangeB=300)) -> (Output_Get:double)
   > n22 = Branch(Condition=Get bIsCrouching())
   > "Set Crouching forward of  backwards Alpha"("Crouching forward of  backwards Alpha"=@n21)
   "Set Crouching forward of  backwards Alpha"(execute=@n22.else, "Crouching forward of  backwards Alpha"=0)
   Event OnCharacterCustomizationLoaded()
   > Set Body Mesh(Body Mesh:Skeletal Mesh Component Object Reference=GetOwnerSkeletalMeshComponent()) -> (Output_Get:Skeletal Mesh Component Object Reference)
   Set Crouching to the left Alpha(execute=@n23.else, Crouching to the left Alpha=0)
   Set Crouching to the left Alpha(execute=@n23.then, Crouching to the left Alpha=@n19)
   n23 = Branch(execute=@n17.then, Condition=Get bIsCrouching())
  ]}
]
variables [
 Body Mesh {type=SkeletalMeshComponent}
 "Crouching forward of  backwards Alpha" {type=number}
 Crouching RArmRaise {type=number}
 Crouching to the left Alpha {type=number}
 Delta time X {type=number}
 Head Tilt To Right Alpha {type=number}
 HeadLookUpAlpha {type=number}
 Is Not Crouching Alpha {type=number}
 Jumping or Falling Check {type=boolean}
 Jumping to the Left Alpha {type=number}
 Jumping to the Right Alpha {type=number}
 "L Arm Pivot Euler Y 0 To -16 " {type=number}
 L Arm Raise 60 To 105 {type=number}
 LeftLegBendEulerY_Alpha {type=number}
 Moving Backwards Alpha {type=number}
 Moving Left or Right Alpha {type=number}
 PouchTransformAlpha {type=number}
 R Arm Raise 45 to 95 {type=number}
 R Arm Raise 95 to 135 {type=number}
 R Arm Raise While Targeting {type=number}
 Spine rotation {type=number}
]
